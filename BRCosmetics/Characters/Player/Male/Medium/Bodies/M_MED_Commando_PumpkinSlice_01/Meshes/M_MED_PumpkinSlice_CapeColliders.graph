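# Physics asset for the PumpkinSlice Commando medium male body, named as the colliders for its cape.
# Capsules cover the pelvis, spine, neck and clavicles, with overlapping capsules on spine_02 and spine_05, and a sphere covers the head.
# Long capsules on the two back dyn_skirt bones add collision volume for the skirt.
# Arm and leg bodies have no shapes, and no constraints exist, so it provides collision shapes without jointed ragdoll setup.
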PhysicsAsset /BRCosmetics/Characters/Player/Male/Medium/Bodies/M_MED_Commando_PumpkinSlice_01/Meshes/M_MED_PumpkinSlice_CapeColliders
bodies {
 calf_l []
 calf_r []
 clavicle_l [{shapeType=Capsule, length=7.7636, radius=11.1427, center=<9.9642 -0.7741 -1.5923>, rotation=<rot 75.9352 -0.0001 175.5622>}]
 clavicle_r [{shapeType=Capsule, length=7.7636, radius=11.1427, center=<-9.9642 0.7741 1.5923>, rotation=<rot 75.9351 -0.0001 -4.4378>}]
 dyn_skirt_l_back_01 [{
   shapeType=Capsule
   length=49
   radius=12.25
   shapeName=dyn_skirt_l_back_01_capsule
   center=<2.3451 -1.6108 5.4316>
   rotation=<rot 90 180 -179.9998>
  }]
 dyn_skirt_r_back_01 [{shapeType=Capsule, length=47.5, radius=12, shapeName=dyn_skirt_r_back_01_capsule, center=<0 1.8115 3.5506>, rotation=<rot 90 0.0003 0.0001>}]
 foot_l []
 foot_r []
 head [{shapeType=Sphere, radius=21.0625, center=<8.138 -0.174 0>}]
 lowerarm_l []
 lowerarm_r []
 neck_02 [{shapeType=Capsule, length=13.4375, radius=7.3175, center=<-0.2642 -0.3408 0>, rotation=<rot 90 -171.1193 -175.3978>}]
 pelvis [{shapeType=Capsule, length=6.9062, radius=15.2888, center=<-2.6215 -1.3695 0>}]
 spine_01 [{shapeType=Capsule, length=10.7187, radius=12.4138, center=<-2.6215 -0.7156 0>}]
 spine_02 [
  {shapeType=Capsule, length=10.5937, radius=10.945, center=<-2.6215 0.4993 0.0001>}
  {shapeType=Capsule, length=29.875, radius=11.1013, center=<2.5024 0.4992 3.9825>, rotation=<rot 90 0.0049 0.0049>}
  {shapeType=Capsule, length=29.875, radius=11.1013, center=<2.4704 0.4991 -3.9825>, rotation=<rot 90 -0.0396 -0.0396>}
 ]
 spine_03 [{shapeType=Capsule, length=9.1562, radius=10.7575, center=<-1.953 0.1364 0.0001>}]
 spine_04 [{shapeType=Capsule, length=14.6562, radius=13.6325, center=<2.1467 -1.2748 0.0002>}]
 spine_05 [
  {shapeType=Capsule, length=17.8437, radius=17.195, center=<-1.7514 2.4879 0>}
  {shapeType=Capsule, length=12.7812, radius=17.0625, center=<0.7875 3.1143 0>}
 ]
 thigh_l []
 thigh_r []
 upperarm_l []
 upperarm_r []
}
constraints []
